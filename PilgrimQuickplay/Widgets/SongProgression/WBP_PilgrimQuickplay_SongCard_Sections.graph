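# Quickplay song progression card widget showing a player's result for a song section: title, stars, accuracy, difficulty, and friend and global leaderboard ranks.
# SetSongCardStats is the main function, fanning inputs out to title, percentage and rank helpers plus child difficulty and star bar widgets, the latter through a deprecated setup call.
# A visibility switcher shows either rounded accuracy or a full-combo icon.
# The event graph holds only unwired events.
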WidgetBlueprint /PilgrimQuickplay/Widgets/SongProgression/WBP_PilgrimQuickplay_SongCard_Sections
graphs [
 EventGraph {nodes [
   Event PreConstruct()
   Event Construct()
   Event Tick()
  ]}
 SetPercentageText {nodes [
   n1 = SetPercentageText!() -> (then, Percentage:double=~, FullCombo:bool=~) {is=title K2Node_FunctionEntry}
   > n2 = Branch(Condition=@n1.FullCombo)
   > SetActiveWidgetIndex(
    self:Common Visibility Switcher Object Reference=Get FullCombo_Switcher() -> (FullCombo_Switcher:Common Visibility Switcher Object Reference)
    Index=1
   )
   "SetText (Text)"(
    execute=@n2.else
    self=Get PercentageText() -> (PercentageText:Common Text Object Reference)
    InText=Format Text!(Format="LOCTEXT(\"{Accuracy}%\")", Accuracy:int=Round(A=@n1) {is=title K2Node_CallFunction}) -> (Result)
   )
   > SetActiveWidgetIndex(self=Get FullCombo_Switcher(), Index=0)
  ]}
 SetPlayerRank {nodes [
   n1 = SetPlayerRank!() -> (then, PilgrimLeaderboardEntry:Pilgrim Leaderboard Entry Structure) {is=title K2Node_FunctionEntry}
   n2 = Break Pilgrim Leaderboard Entry!(PilgrimLeaderboardEntry:"Pilgrim Leaderboard Entry Structure (by ref)"=@n1) -> (
    bHasLocalPlayer:bool=~
    PlayerAccountIdToDisplayNameMap:Map of Strings to Strings
    Rank:int
    Percentile:float=~
    ScoreData:Pilgrim Leaderboard Score Data Structure
    DateTime:Date Time Structure
    FriendRank:int
   ) {is=title K2Node_BreakStruct}
   > "SetText (Text)"(self=Get GlobalRank() -> (GlobalRank:Common Text Object Reference), InText="To Text (Integer)"(Value=@n2.Rank))
   > "SetText (Text)"(self=Get FriendRank() -> (FriendRank:Common Text Object Reference), InText="To Text (Integer)"(Value=@n2.FriendRank))
  ]}
 SetSongCardStats {nodes [
   n1 = SetSongCardStats!() -> (
    then
    SongCardTitle:Text
    Accuracy:double=~
    FullCombo:bool=~
    Instrument:EPilgrimTrackType Enum=TrackGuitar
    StarCount:int
    Difficulty:EPilgrimSongDifficulty Enum=DifficultyEasy
    PilgrimLeaderboardEntry:Pilgrim Leaderboard Entry Structure
   ) {is=title K2Node_FunctionEntry}
   > n2 = Sequence!(execute) -> (then_0, then_1, then_2:Exec, then_3:Exec, then_4:Exec)
   > SetSongCardTitle!(execute, self, SongCardTitle:Text=@n1) -> (then)
   SetPercentageText!(execute=@n2.then_1, self, Percentage:double=@n1.Accuracy, FullCombo:bool=@n1.FullCombo) -> (then) {is=title K2Node_CallFunction}
   SetupStarsResults_DEPRECATED!(
    execute=@n2.then_2
    self:WBP Pilgrim Quickplay Local Star Bar Object Reference=Get WBP_PilgrimQuickplay_LocalStarBar() -> (WBP_PilgrimQuickplay_LocalStarBar:WBP Pilgrim Quickplay Local Star Bar Object Reference)
    Instrument:EPilgrimTrackType Enum=@n1.Instrument
    StarCount:int=@n1.StarCount
   ) -> (then)
   SetDifficulty!(
    execute=@n2.then_3
    self:WBP Pilgrim Results Difficulty Object Reference=Get WBP_Pilgrim_Results_Difficulty() -> (WBP_Pilgrim_Results_Difficulty:WBP Pilgrim Results Difficulty Object Reference)
    Difficulty:EPilgrimSongDifficulty Enum=@n1.Difficulty
   ) -> (then)
   SetPlayerRank!(execute=@n2.then_4, self, PilgrimLeaderboardEntry:Pilgrim Leaderboard Entry Structure=@n1.PilgrimLeaderboardEntry) -> (then) {is=title K2Node_CallFunction}
  ]}
 SetSongCardTitle {nodes [
   n1 = SetSongCardTitle!() -> (then, SongCardTitle:Text) {is=title K2Node_FunctionEntry}
   > "SetText (Text)"(self=Get SongCardTitle() -> (SongCardTitle:Common Text Object Reference), InText=@n1)
  ]}
]
variables []
widgets {
 info {inheritedWidgetCount=0, parentClass=&CommonUserWidget, rootWidgetClass=&CommonBorder, widgetCount=16}
 widgets [
  CommonBorder_1 {
   widgetClassPath=&CommonBorder
   widget=CommonBorder_1
   properties {
    background {
     bIsDynamicallyLoaded=false
     drawAs=RoundedBox
     imageType=FullColor
     mirroring=NoMirror
     resourceName=None
     resourceObject=/PilgrimQuickplay/Widgets/Leaderboards/Materials/M_Leaderboard_CurrentUser
     tiling=NoTile
     imageSize=<32 32>
     margin=<ltrb 0 0 0 0>
     outlineSettings {
      bUseBrushTransparency=false
      roundingType=FixedRadius
      width=0
      color {colorUseRule=UseColor_Specified, specifiedColor=<rgba 0 0 0 0>}
      cornerRadii=<15 15 15 15>
     }
     tintColor {colorUseRule=UseColor_Specified, specifiedColor=<rgba 1 1 1 1>}
     uVRegion {bIsValid=false, max=<0 0>, min=<0 0>}
    }
    padding=<ltrb 0 0 0 0>
   }
  }
  VerticalBox_25 {
   widgetClassPath=&VerticalBox
   parent=CommonBorder_1
   widget=VerticalBox_25
   slot {class=BorderSlot, properties {padding=<ltrb 0 0 0 0>}}
  }
  SongCardTitle {
   widgetClassPath=&CommonTextBlock
   bIsVariable=true
   parent=VerticalBox_25
   widget=SongCardTitle
   properties {
    text=Season 00
    font {
     bForceMonospaced=false
     bMaterialIsStencil=false
     fontMaterial=null
     fontObject=&HeadingNow
     letterSpacing=0
     monospacedWidth=1
     size=18
     skewAmount=0
     typefaceFontName=74 Regular
     outlineSettings {
      bApplyOutlineToDropShadows=false
      bMiteredCorners=false
      bSeparateFillAlpha=false
      outlineMaterial=null
      outlineSize=0
      outlineColor=<rgba 1 1 1 1>
     }
    }
    shadowOffset=<0 0>
   }
   slot {class=VerticalBoxSlot, properties {horizontalAlignment=HAlign_Center, padding=<ltrb 0 0 0 5>}}
  }
  WBP_PilgrimQuickplay_LocalStarBar {
   widgetClassPath=/PilgrimQuickplay/Widgets/WBP_PilgrimQuickplay_LocalStarBar
   bIsVariable=true
   parent=VerticalBox_25
   widget=WBP_PilgrimQuickplay_LocalStarBar
   properties {in User Specified Scale=0.4, scale Stretch=UserSpecified}
   slot {class=VerticalBoxSlot, properties {horizontalAlignment=HAlign_Center, verticalAlignment=VAlign_Center, padding=<ltrb 0 0 0 5>}}
  }
  HorizontalBox_58 {
   widgetClassPath=&HorizontalBox
   parent=VerticalBox_25
   widget=HorizontalBox_58
   slot {class=VerticalBoxSlot, properties {horizontalAlignment=HAlign_Center, padding=<ltrb 0 0 0 5>}}
  }
  FullCombo_Switcher {
   widgetClassPath=&CommonVisibilitySwitcher
   bIsVariable=true
   parent=HorizontalBox_58
   widget=FullCombo_Switcher
   slot {class=HorizontalBoxSlot, properties {horizontalAlignment=HAlign_Center, verticalAlignment=VAlign_Center, padding=<ltrb 5 0 0 0>}}
  }
  PercentageText {
   widgetClassPath=&CommonTextBlock
   bIsVariable=true
   parent=FullCombo_Switcher
   widget=PercentageText
   properties {
    text=99%
    visibility=Collapsed
    colorAndOpacity {colorUseRule=UseColor_Specified, specifiedColor=<rgba 0.4125 0.5089 0.9647 1>}
    font {
     bForceMonospaced=false
     bMaterialIsStencil=false
     fontMaterial=null
     fontObject=&HeadingNow
     letterSpacing=0
     monospacedWidth=1
     size=18
     skewAmount=0
     typefaceFontName=74 Regular
     outlineSettings {
      bApplyOutlineToDropShadows=false
      bMiteredCorners=false
      bSeparateFillAlpha=false
      outlineMaterial=null
      outlineSize=0
      outlineColor=<rgba 1 1 1 1>
     }
    }
    shadowOffset=<0 0>
   }
   slot {class=CommonVisibilitySwitcherSlot, properties {horizontalAlignment=HAlign_Center, verticalAlignment=VAlign_Center}}
  }
  FullComboIndicator {
   widgetClassPath=/PilgrimQuickplay/Widgets/WBP_PilgrimQuickplay_FullComboIndicator
   bIsVariable=true
   parent=FullCombo_Switcher
   widget=FullComboIndicator
   properties {showIconOnly=true, visibility=Collapsed}
   slot {class=CommonVisibilitySwitcherSlot, properties {horizontalAlignment=HAlign_Center, verticalAlignment=VAlign_Center}}
  }
  Spacer {widgetClassPath=&Spacer, parent=HorizontalBox_58, widget=Spacer, properties {size=<15 1>}, slot {class=HorizontalBoxSlot}}
  WBP_Pilgrim_Results_Difficulty {
   widgetClassPath=/PilgrimQuickplay/Widgets/Results/WBP_Pilgrim_Results_Difficulty
   bIsVariable=true
   parent=HorizontalBox_58
   widget=WBP_Pilgrim_Results_Difficulty
   slot {class=HorizontalBoxSlot}
  }
  HorizontalBox_121 {
   widgetClassPath=&HorizontalBox
   parent=VerticalBox_25
   widget=HorizontalBox_121
   slot {class=VerticalBoxSlot, properties {horizontalAlignment=HAlign_Center}}
  }
  FriendIcon {
   widgetClassPath=&Image
   bIsVariable=true
   parent=HorizontalBox_121
   widget=FriendIcon
   properties {
    visibility=SelfHitTestInvisible
    brush {
     bIsDynamicallyLoaded=false
     drawAs=Image
     imageType=FullColor
     mirroring=NoMirror
     resourceName=None
     resourceObject=/Game/UI/Foundation/Shell/Textures/T-Icon-Friends-L
     tiling=NoTile
     imageSize=<32 32>
     margin=<ltrb 0 0 0 0>
     outlineSettings {
      bUseBrushTransparency=false
      roundingType=HalfHeightRadius
      width=0
      color {colorUseRule=UseColor_Specified, specifiedColor=<rgba 0 0 0 0>}
      cornerRadii=<0 0 0 0>
     }
     tintColor {colorUseRule=UseColor_Specified, specifiedColor=<rgba 1 1 1 1>}
     uVRegion {bIsValid=false, max=<0 0>, min=<0 0>}
    }
    colorAndOpacity=<rgba 0.3916 0.4852 0.9387 1>
   }
   slot {class=HorizontalBoxSlot}
  }
  FriendRank {
   widgetClassPath=&CommonTextBlock
   bIsVariable=true
   parent=HorizontalBox_121
   widget=FriendRank
   properties {
    text=999
    font {
     bForceMonospaced=false
     bMaterialIsStencil=false
     fontMaterial=null
     fontObject=&HeadingNow
     letterSpacing=0
     monospacedWidth=1
     size=18
     skewAmount=0
     typefaceFontName=74 Regular
     outlineSettings {
      bApplyOutlineToDropShadows=false
      bMiteredCorners=false
      bSeparateFillAlpha=false
      outlineMaterial=null
      outlineSize=0
      outlineColor=<rgba 1 1 1 1>
     }
    }
    shadowOffset=<0 0>
   }
   slot {class=HorizontalBoxSlot, properties {verticalAlignment=VAlign_Center}}
  }
  Spacer_73 {widgetClassPath=&Spacer, parent=HorizontalBox_121, widget=Spacer_73, properties {size=<15 1>}, slot {class=HorizontalBoxSlot}}
  GlobalIcon {
   widgetClassPath=&Image
   bIsVariable=true
   parent=HorizontalBox_121
   widget=GlobalIcon
   properties {
    visibility=SelfHitTestInvisible
    brush {
     bIsDynamicallyLoaded=false
     drawAs=Image
     imageType=FullColor
     mirroring=NoMirror
     resourceName=None
     resourceObject=/Game/UI/Friends_UI/Social/Globe_PlatformIcon_32x
     tiling=NoTile
     imageSize=<32 32>
     margin=<ltrb 0 0 0 0>
     outlineSettings {
      bUseBrushTransparency=false
      roundingType=HalfHeightRadius
      width=0
      color {colorUseRule=UseColor_Specified, specifiedColor=<rgba 0 0 0 0>}
      cornerRadii=<0 0 0 0>
     }
     tintColor {colorUseRule=UseColor_Specified, specifiedColor=<rgba 1 1 1 1>}
     uVRegion {bIsValid=false, max=<0 0>, min=<0 0>}
    }
    colorAndOpacity=<rgba 0.3916 0.4852 0.9387 1>
   }
   slot {class=HorizontalBoxSlot}
  }
  GlobalRank {
   widgetClassPath=&CommonTextBlock
   bIsVariable=true
   parent=HorizontalBox_121
   widget=GlobalRank
   properties {
    text=999
    font {
     bForceMonospaced=false
     bMaterialIsStencil=false
     fontMaterial=null
     fontObject=&HeadingNow
     letterSpacing=0
     monospacedWidth=1
     size=18
     skewAmount=0
     typefaceFontName=74 Regular
     outlineSettings {
      bApplyOutlineToDropShadows=false
      bMiteredCorners=false
      bSeparateFillAlpha=false
      outlineMaterial=null
      outlineSize=0
      outlineColor=<rgba 1 1 1 1>
     }
    }
    shadowOffset=<0 0>
   }
   slot {class=HorizontalBoxSlot}
  }
 ]
}
