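# Song card widget for Pilgrim Quickplay song progression, laying out a record image above a song level label.
# A rounded CommonBorder, using the leaderboard current-user material as background, holds a centered vertical stack of RecordImage and SongLevelText.
# The image brush is empty and the text is a "Level ##" placeholder; the event graph has only unconnected default events, so nothing here fills them in.
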WidgetBlueprint /PilgrimQuickplay/Widgets/SongProgression/WBP_PilgrimQuickplay_SongCard_SongLevel
graphs [EventGraph {nodes [
   Event PreConstruct()
   Event Construct()
   Event Tick()
  ]}]
variables []
widgets {
 info {inheritedWidgetCount=0, parentClass=&CommonUserWidget, rootWidgetClass=&CommonBorder, widgetCount=4}
 widgets [
  CommonBorder_1 {
   widgetClassPath=&CommonBorder
   widget=CommonBorder_1
   properties {
    background {
     bIsDynamicallyLoaded=false
     drawAs=RoundedBox
     imageType=FullColor
     mirroring=NoMirror
     resourceName=None
     resourceObject=/PilgrimQuickplay/Widgets/Leaderboards/Materials/M_Leaderboard_CurrentUser
     tiling=NoTile
     imageSize=<32 32>
     margin=<ltrb 0 0 0 0>
     outlineSettings {
      bUseBrushTransparency=false
      roundingType=FixedRadius
      width=0
      color {colorUseRule=UseColor_Specified, specifiedColor=<rgba 0 0 0 0>}
      cornerRadii=<15 15 15 15>
     }
     tintColor {colorUseRule=UseColor_Specified, specifiedColor=<rgba 1 1 1 1>}
     uVRegion {bIsValid=false, max=<0 0>, min=<0 0>}
    }
    padding=<ltrb 0 0 0 0>
   }
  }
  VerticalBox_33 {
   widgetClassPath=&VerticalBox
   parent=CommonBorder_1
   widget=VerticalBox_33
   slot {class=BorderSlot, properties {padding=<ltrb 0 0 0 0>}}
  }
  RecordImage {
   widgetClassPath=&Image
   bIsVariable=true
   parent=VerticalBox_33
   widget=RecordImage
   properties {brush {
     bIsDynamicallyLoaded=false
     drawAs=Image
     imageType=NoImage
     mirroring=NoMirror
     resourceName=None
     resourceObject=null
     tiling=NoTile
     imageSize=<75 75>
     margin=<ltrb 0 0 0 0>
     outlineSettings {
      bUseBrushTransparency=false
      roundingType=HalfHeightRadius
      width=0
      color {colorUseRule=UseColor_Specified, specifiedColor=<rgba 0 0 0 0>}
      cornerRadii=<0 0 0 0>
     }
     tintColor {colorUseRule=UseColor_Specified, specifiedColor=<rgba 1 1 1 1>}
     uVRegion {bIsValid=false, max=<0 0>, min=<0 0>}
    }}
   slot {class=VerticalBoxSlot, properties {horizontalAlignment=HAlign_Center, padding=<ltrb 0 25 0 0>}}
  }
  SongLevelText {
   widgetClassPath=&CommonTextBlock
   parent=VerticalBox_33
   widget=SongLevelText
   properties {
    text="Level ##"
    font {
     bForceMonospaced=false
     bMaterialIsStencil=false
     fontMaterial=null
     fontObject=&HeadingNow
     letterSpacing=0
     monospacedWidth=1
     size=18
     skewAmount=0
     typefaceFontName=74 Regular
     outlineSettings {
      bApplyOutlineToDropShadows=false
      bMiteredCorners=false
      bSeparateFillAlpha=false
      outlineMaterial=null
      outlineSize=0
      outlineColor=<rgba 1 1 1 1>
     }
    }
    shadowOffset=<0 0>
   }
   slot {class=VerticalBoxSlot, properties {horizontalAlignment=HAlign_Center, verticalAlignment=VAlign_Center, padding=<ltrb 0 10 0 0>}}
  }
 ]
}
