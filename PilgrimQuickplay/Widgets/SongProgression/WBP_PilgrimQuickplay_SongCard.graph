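# WBP_PilgrimQuickplay_SongCard is a CommonUserWidget that forms a song card in the Pilgrim Quickplay song progression UI.
# A horizontal box lays out a song level plaque beside Season and All-Time sections, split by spacers, with the plaque given less fill width.
# Both sections reuse one Sections widget class.
# It is layout only: the event stubs are empty and it declares no variables of its own.
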WidgetBlueprint /PilgrimQuickplay/Widgets/SongProgression/WBP_PilgrimQuickplay_SongCard
graphs [EventGraph {nodes [
   Event PreConstruct()
   Event Construct()
   Event Tick()
  ]}]
variables []
widgets {
 info {inheritedWidgetCount=0, parentClass=&CommonUserWidget, rootWidgetClass=&HorizontalBox, widgetCount=6}
 widgets [
  HorizontalBox_15 {widgetClassPath=&HorizontalBox, widget=HorizontalBox_15}
  SongLevelPlaque {
   widgetClassPath=&/WBP_PilgrimQuickplay_SongCard_SongLevel
   bIsVariable=true
   parent=HorizontalBox_15
   widget=SongLevelPlaque
   slot {class=HorizontalBoxSlot, properties {size {sizeRule=Fill, value=0.4}}}
  }
  Spacer {widgetClassPath=&Spacer, parent=HorizontalBox_15, widget=Spacer, properties {size=<15 1>}, slot {class=HorizontalBoxSlot}}
  Season_Section {
   widgetClassPath=&/WBP_PilgrimQuickplay_SongCard_Sections
   bIsVariable=true
   parent=HorizontalBox_15
   widget=Season_Section
   slot {class=HorizontalBoxSlot, properties {size {sizeRule=Fill, value=1}}}
  }
  Spacer_70 {widgetClassPath=&Spacer, parent=HorizontalBox_15, widget=Spacer_70, properties {size=<15 1>}, slot {class=HorizontalBoxSlot}}
  AllTime_Section {
   widgetClassPath=&WBP_PilgrimQuickplay_SongCard_Sections
   bIsVariable=true
   parent=HorizontalBox_15
   widget=AllTime_Section
   slot {class=HorizontalBoxSlot, properties {size {sizeRule=Fill, value=1}}}
  }
 ]
}
